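# Modular Control Rig module for folding a quadruped companion's wing with a single "Wing Main" control.
# Construction resolves Hand and Wing A–D connectors, spawns the control under Hand, and adds a "_driver" bone under Hand at each digit's transform.
# Forward solve projects Wing Main's motion onto the drivers, fully for A and by tunable weights for B–D, fanning the fold; digit bones are not written here.
# Backward solve snaps Wing Main to Hand.
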ControlRigBlueprint /CosmeticCompanions/Assets/Quadruped/TalonScone/Meshes/CRM_WingFold_v01
graphs [
 ConstructionGraph {nodes [
   RigUnit_PrepareForExecution = &RigUnit_PrepareForExecution()
   RigVMFunction_Sequence = RigVMAggregateNode(ExecuteContext=@RigUnit_PrepareForExecution.ExecutePin)
   RigVMDispatchNode(ExecuteContext=@RigVMFunction_Sequence.A, Array=RigVMVariableNode() {variable=Wing Digits}.Value) {template="DISPATCH_RigVMDispatch_ArrayReset(io Array)"}
   > RigVMDispatchNode(Array=RigVMVariableNode() {variable=Wing Digits Virtual}.Value) {template="DISPATCH_RigVMDispatch_ArrayReset(io Array)"}
   RigVMVariableNode(
    Value=/Script/ControlRig.RigUnit_ResolveConnector(Connector=(Type=Connector,Name="Hand"), SkipSocket=False) {is=struct RigVMUnitNode}.Result
    ExecuteContext=@RigVMFunction_Sequence.B
   ) {variable=Hand}
   n1 = &RigUnit_ResolveConnector(Connector=(Type=Connector,Name="Wing A"), SkipSocket=False)
   > RigVMVariableNode(Value=@n1.Result) {variable=Wing A}
   > RigVMDispatchNode(Element=@n1.Result, Array=RigVMVariableNode() {variable=Wing Digits}.Value) {template="DISPATCH_RigVMDispatch_ArrayAdd(io Array,in Element,out Index)"}
   n2 = &RigUnit_ResolveConnector(Connector=(Type=Connector,Name="Wing B"), SkipSocket=False)
   > RigVMVariableNode(Value=@n2.Result) {variable=Wing B}
   > RigVMDispatchNode(Element=@n2.Result, Array=RigVMVariableNode() {variable=Wing Digits}.Value) {template="DISPATCH_RigVMDispatch_ArrayAdd(io Array,in Element,out Index)"}
   n3 = &RigUnit_ResolveConnector(Connector=(Type=Connector,Name="Wing C"), SkipSocket=False)
   > RigVMVariableNode(Value=@n3.Result) {variable=Wing C}
   > RigVMDispatchNode(Element=@n3.Result, Array=RigVMVariableNode() {variable=Wing Digits}.Value) {template="DISPATCH_RigVMDispatch_ArrayAdd(io Array,in Element,out Index)"}
   n4 = &RigUnit_ResolveConnector(Connector=(Type=Connector,Name="Wing D"), SkipSocket=False)
   > RigVMVariableNode(Value=@n4.Result) {variable=Wing D}
   > RigVMDispatchNode(Element=@n4.Result, Array=RigVMVariableNode() {variable=Wing Digits}.Value) {template="DISPATCH_RigVMDispatch_ArrayAdd(io Array,in Element,out Index)"}
   n5 = RigVMVariableNode() {variable=Control Scale}
   SpawnControl = /Script/ControlRig.RigUnit_HierarchyAddControlTransform(
    InitialValue=(Rotation=(X=0,Y=0,Z=0,W=1),Translation=(X=0,Y=0,Z=0),Scale3D=(X=1,Y=1,Z=1))
    Name=Wing Main
    OffsetSpace=LocalSpace
    OffsetTransform=(Rotation=(X=0,Y=0,Z=0,W=1),Translation=(X=0,Y=0,Z=0),Scale3D=(X=1,Y=1,Z=1))
    Settings="(InitialSpace=LocalSpace,bUsePreferredRotationOrder=False,PreferredRotationOrder=YZX,Limits=(LimitTranslationX=(bMinimum=False,bMaximum=False),LimitTranslationY=(bMinimum=False,bMaximum=False),LimitTranslationZ=(bMinimum=False,bMaximum=False),LimitPitch=(b ...(truncated)"
    ExecutePin=@RigVMFunction_Sequence.C
    Settings.Shape.Color=RigVMVariableNode() {variable=Color}.Value
    Settings.Shape.Transform.Translation=RigVMVariableNode() {variable=Translate}.Value
    Settings.Shape.Transform.Scale3D.X=@n5.Value
    Settings.Shape.Transform.Scale3D.Y=@n5.Value
    Settings.Shape.Transform.Scale3D.Z=@n5.Value
    Parent=RigVMVariableNode() {variable=Hand}.Value
   ) {
    is=struct RigVMUnitNode
    template="SpawnControl::Execute(in InitialValue,in Settings,in OffsetTransform,in OffsetSpace,in Parent,in Name,out Item)"
   }
   > RigVMVariableNode(Value=@SpawnControl.Item) {variable=Wing Main}
   > For_Each = RigVMDispatchNode(Array=RigVMVariableNode() {variable=Wing Digits}.Value) {template="DISPATCH_RigVMDispatch_ArrayIterator(in Array,out Element,out Index,out Count,out Ratio)"}
   Concat = /Script/RigVM.RigVMFunction_NameConcat(B=_driver, A=@For_Each.Element.Name) {is=struct RigVMUnitNode, template="Concat::Execute(in A,in B,out Result)"}
   > n6 = /Script/ControlRig.RigUnit_HierarchyAddBone(
    Space=GlobalSpace
    Name=@Concat.Result
    Transform=&RigUnit_GetTransform(bInitial=False, Item=@For_Each.Element).Transform
    Parent=RigVMVariableNode() {variable=Hand}.Value
   ) {is=struct RigVMUnitNode}
   > RigVMDispatchNode(Element=@n6.Item, Array=RigVMVariableNode() {variable=Wing Digits Virtual}.Value) {template="DISPATCH_RigVMDispatch_ArrayAdd(io Array,in Element,out Index)"}
  ]}
 RigVMFunctionLibrary {links [], nodes []}
 RigVMFunction_Sequence_1_ContainedGraph {nodes [
   n1 = RigVMFunctionEntryNode()
   RigVMFunction_Sequence = &RigVMFunction_Sequence(ExecuteContext=@n1.ExecuteContext)
   n2 = &RigVMFunction_Sequence(ExecuteContext=@RigVMFunction_Sequence.B)
   RigVMFunctionReturnNode(A=@RigVMFunction_Sequence.A, B=@n2.A, C=@n2.B)
  ]}
 RigVMFunction_Sequence_1_ContainedGraph {nodes [
   n1 = RigVMFunctionEntryNode()
   RigVMFunction_Sequence = &RigVMFunction_Sequence(ExecuteContext=@n1.ExecuteContext)
   n2 = &RigVMFunction_Sequence(ExecuteContext=@RigVMFunction_Sequence.B)
   n3 = &RigVMFunction_Sequence(ExecuteContext=@n2.B)
   RigVMFunctionReturnNode(A=@RigVMFunction_Sequence.A, B=@n2.A, C=@n3.A, D=@n3.B)
  ]}
 RigVMModel {nodes [
   RigUnit_BeginExecution = &RigUnit_BeginExecution()
   RigVMFunction_Sequence = RigVMAggregateNode(ExecuteContext=@RigUnit_BeginExecution.ExecutePin)
   n1 = RigVMDispatchNode(Index=0, Array=RigVMVariableNode() {variable=Wing Digits Virtual}.Value) {template="DISPATCH_RigVMDispatch_ArrayGetAtIndex(in Array,in Index,out Element)"}
   n2 = RigVMVariableNode() {variable=Wing Main}
   &RigUnit_SetTransform(
    Weight=1
    bInitial=False
    Item=@n1.Element
    Value=/Script/ControlRig.RigUnit_ProjectTransformToNewParent(bChildInitial=True, bNewParentInitial=False, bOldParentInitial=True, Child=@n1.Element, NewParent=@n2.Value, OldParent=@n2.Value) {is=struct RigVMUnitNode}.Transform
    ExecutePin=@RigVMFunction_Sequence.A
   )
   n3 = RigVMDispatchNode(Index=1, Array=RigVMVariableNode() {variable=Wing Digits Virtual}.Value) {template="DISPATCH_RigVMDispatch_ArrayGetAtIndex(in Array,in Index,out Element)"}
   n4 = RigVMVariableNode() {variable=Wing Main}
   > &RigUnit_SetTransform(
    bInitial=False
    Item=@n3.Element
    Value=&RigUnit_ProjectTransformToNewParent(bChildInitial=True, bNewParentInitial=False, bOldParentInitial=True, Child=@n3.Element, NewParent=@n4.Value, OldParent=@n4.Value).Transform
    Weight=RigVMVariableNode() {variable=Finger B Rot}.Value
   )
   n5 = RigVMDispatchNode(Index=2, Array=RigVMVariableNode() {variable=Wing Digits Virtual}.Value) {template="DISPATCH_RigVMDispatch_ArrayGetAtIndex(in Array,in Index,out Element)"}
   n6 = RigVMVariableNode() {variable=Wing Main}
   > &RigUnit_SetTransform(
    bInitial=False
    Item=@n5.Element
    Value=&RigUnit_ProjectTransformToNewParent(bChildInitial=True, bNewParentInitial=False, bOldParentInitial=True, Child=@n5.Element, NewParent=@n6.Value, OldParent=@n6.Value).Transform
    Weight=RigVMVariableNode() {variable=Finger C Rot}.Value
   )
   n7 = RigVMDispatchNode(Index=3, Array=RigVMVariableNode() {variable=Wing Digits Virtual}.Value) {template="DISPATCH_RigVMDispatch_ArrayGetAtIndex(in Array,in Index,out Element)"}
   n8 = RigVMVariableNode() {variable=Wing Main}
   > &RigUnit_SetTransform(
    bInitial=False
    Item=@n7.Element
    Value=&RigUnit_ProjectTransformToNewParent(bChildInitial=True, bNewParentInitial=False, bOldParentInitial=True, Child=@n7.Element, NewParent=@n8.Value, OldParent=@n8.Value).Transform
    Weight=RigVMVariableNode() {variable=Finger D Rot}.Value
   )
  ]}
 RigVMModel Rig {nodes [
   n1 = /Script/ControlRig.RigUnit_InverseExecution() {is=struct RigVMUnitNode}
   &RigUnit_SetTransform(
    Weight=1
    bInitial=False
    Value=&RigUnit_GetTransform(bInitial=False, Item=RigVMVariableNode() {variable=Hand}.Value).Transform
    ExecutePin=@n1.ExecutePin
    Item=RigVMVariableNode() {variable=Wing Main}.Value
   )
  ]}
]
hierarchy [
 Hand {type=Connector, loc=<0 0 0>, rot=<0 0 0 1>}
 Root {type=Socket, loc=<0 0 0>, rot=<0 0 0 1>}
 Wing B {type=Connector, loc=<0 0 0>, rot=<0 0 0 1>}
 Wing C {type=Connector, loc=<0 0 0>, rot=<0 0 0 1>}
 Wing D {type=Connector, loc=<0 0 0>, rot=<0 0 0 1>}
 Wing A {type=Connector, loc=<0 0 0>, rot=<0 0 0 1>}
]
variables [
 Color {type=FLinearColor, default=(R=1,G=0,B=0,A=1)}
 Control Scale {type=double, default=0}
 Finger B Rot {type=float, default=0.85}
 Finger C Rot {type=float, default=0.65}
 Finger D Rot {type=float, default=0.3}
 Hand {type=FRigElementKey, default=(Type=Connector,Name="")}
 Translate {type=FVector, default=(X=0,Y=0,Z=0)}
 Wing A {type=FRigElementKey, default=(Type=Connector,Name="")}
 Wing B {type=FRigElementKey, default=(Type=Connector,Name="")}
 Wing C {type=FRigElementKey, default=(Type=Connector,Name="")}
 Wing D {type=FRigElementKey, default=(Type=Connector,Name="")}
 Wing Digits {type="TArray<FRigElementKey>"}
 Wing Digits Virtual {type="TArray<FRigElementKey>"}
 Wing Main {type=FRigElementKey, default=(Type=Bone,Name="")}
]
