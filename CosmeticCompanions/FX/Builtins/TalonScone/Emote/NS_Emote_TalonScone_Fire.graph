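# Niagara system for the TalonScone cosmetic companion's fire emote, a short directional burst of flame.
# Emitters include MainFire, FiresprintSubUV flipbooks, Embers, Smoke and a Glow with mesh and light, all launched along a shared system velocity cone.
# System age tapers spawning and AgeFade fades particles, while HotdogDist scales spawn rates and light offset.
# User multipliers tune lifetime and velocity, spawn groups layer variants within emitters, and front-end modifiers dim the flames.
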NiagaraSystem /CosmeticCompanions/FX/Builtins/TalonScone/Emote/NS_Emote_TalonScone_Fire
dependencies {
 usedDataInterfaces [
  &NiagaraDataInterfaceCurve
  &NiagaraDataInterfaceColorCurve
  &NiagaraDataInterfaceVector2DCurve
  &NiagaraDataInterfaceSpriteRendererInfo
  &NiagaraDataInterfaceSocketReader
 ]
 usedDynamicInputs [&RandomRangeVector2D, &FNTimeOfDayScalar]
 usedModules [
  &self:SystemSpawnScript.SystemScriptSource.SystemScriptGraph.NiagaraNodeAssignment_6.SetVariables_#1
  &self:SystemSpawnScript.SystemScriptSource.SystemScriptGraph.NiagaraNodeAssignment_5.SetVariables_#2
  &self:SystemSpawnScript.SystemScriptSource.SystemScriptGraph.NiagaraNodeAssignment_0.SetVariables_#3
  &self:SystemSpawnScript.SystemScriptSource.SystemScriptGraph.NiagaraNodeAssignment_4.SetVariables_#4
  &self:SystemSpawnScript.SystemScriptSource.SystemScriptGraph.NiagaraNodeAssignment_2.SetVariables_#5
  &SystemState
  &self:SystemSpawnScript.SystemScriptSource.SystemScriptGraph.NiagaraNodeAssignment_3.SetVariables_#6
  &self:SystemSpawnScript.SystemScriptSource.SystemScriptGraph.NiagaraNodeAssignment_1.SetVariables_#7
 ]
}
emitters {
 Embers {
  emitterUpdateScript [
   &EmitterState(Life Cycle Mode=::NewEnumerator1, Loop Behavior=::NewEnumerator1, Loop Duration=1, UseLoopDelay=false)
   &SpawnRate(
    SpawnRate=&FloatFromCurve(
     FloatCurve=DataInterface() {propertyValues {Curve {keys [<at 0.6 1 RCIM_Cubic>,<at 1 0>]}}}
     CurveIndex=$System.Age
     Scale Curve=&Lerp_Float(A=50, B=80, Alpha=$System.HotdogDist)
    )
    Use Spawn Probability=false
   )
  ]
  particleSpawnScript [
   &FN_SetSpawnGroup(Number of Spawn Groups=2)
   /SetVariables(Particles.ScaleSize:Vector2f=<4 4>) {isSetParameters=true}
   &InitializeParticle(
    Color=<rgba 3 0.7919 0.3503 1>
    Lifetime=&Multiply_Float(A=&RandomRangeFloat(Minimum=0.5, Maximum=1), B=$User.LifeMultiplier)
    Sprite Size Max=<0.7 3>
    Sprite Size Min=<0.5 4>
    Sprite Size Mode=::NewEnumerator2
    UsePositionOffset=true
    Write Material Random=true
   )
   &CameraOffset(Camera Offset Amount=5)
   &AddVelocity(
    Cone Angle=&Multiply_Float(
     A=$System.VelocityConeAngle
     B=&MergeHelper_FloatSwitch(0=1.5, 1=3, 2=?, 3=?, Index=&SpawnGroup(), Num=::NewEnumerator0)
    )
    Cone Axis=$System.VelocityConeAxis
    Module.UseConeFalloff=false
    Module.UseVelDistribution=false
    Override Seed=false
    Velocity Mode=::NewEnumerator2
    Velocity Speed=&Multiply_Float(
     A=&RandomRangeFloat(Minimum=$System.VelocitySpeedMin, Maximum=$System.VelocitySpeedMax)
     B=&MergeHelper_FloatSwitch(0=0.7, 1=0.4, 2=?, 3=?, Index=&SpawnGroup(), Num=::NewEnumerator0)
    )
   ) {module=AddVelocity001}
  ]
  particleUpdateScript [
   &ParticleState()
   /SetVariables(Particles.Glow03_Color=&MakeLinearColorFromVectorAndFloat(
     "Vector (RGB)"=&MakeVectorFromLinearColorRGB(Color=&MultiplyLinearColor(Linear Color=<rgba 1 0.0833 0 0.2>, Scale Factor=&MakeVector4_FromFloat(Float=1)))
     "Float (Alpha)"=&Multiply_Float(A=0.12, B=$System.AgeFade)
    )) {isSetParameters=true}
   /SetVariables(Particles.Glow03_Size=&Multiply_Vector2D_ByFloat(
     Float=&FloatFromCurve(
      FloatCurve=DataInterface() {propertyValues {Curve {keys [
          <at 0 0 RCIM_Cubic>
          <at 0.2 1.3 RCIM_Cubic>
          {arriveTangent=-0.1423, interpMode=RCIM_Cubic, leaveTangent=-0.1423, tangentMode=RCTM_User, time=1, value=0}
         ]}}}
      Scale Curve=10
     )
     Vector2D=&Multiply_Vector2D(A=<0.4 1>, B=$Particles.ScaleSize)
    )) {isSetParameters=true}
   &ScaleColor(
    Scale Alpha=&FloatFromCurve(FloatCurve=DataInterface() {propertyValues {Curve {keys [<at 0.3 1>,<at 0.4 1>,<at 1 1>]}}})
    ScaleA=true
    ScaleRGB=true
   )
   &ScaleSpriteSize(Scale Factor=&Multiply_Vector2D(
     A=&Vector2DFromFloat(Value=&FloatFromCurve(
       FloatCurve=DataInterface() {propertyValues {Curve {keys [
           {arriveTangent=14.1901, interpMode=RCIM_Cubic, leaveTangent=14.1901, tangentMode=RCTM_User, time=0.02, value=0.0598}
           {interpMode=RCIM_Cubic, leaveTangent=-3.0296, tangentMode=RCTM_Break, time=0.2, value=1.2}
           {arriveTangent=-2.8894, interpMode=RCIM_Cubic, leaveTangent=-2.8894, tangentMode=RCTM_User, time=1, value=0.1}
          ]}}}
       Scale Curve=0.4
      ))
     B=$Particles.ScaleSize
    )) {module=ScaleSpriteSize001}
   &CurlNoiseForce(
    Noise Frequency=20
    Noise Strength=&Multiply_Float(
     A=&MergeHelper_FloatSwitch(0=8000, 1=20000, 2=?, 3=?, Index=&SpawnGroup(), Num=::NewEnumerator0)
     B=&FloatFromCurve(
      FloatCurve=DataInterface() {propertyValues {Curve {keys [{arriveTangent=2.1957, interpMode=RCIM_Cubic, leaveTangent=2.1957, tangentMode=RCTM_User, time=0, value=0}, <at 1 1>]}}}
      Scale Curve=&Lerp_Float(A=0.8, B=0.7, Alpha=$System.FrontEndLerp)
     )
    )
    Pan Noise=true
    Pan Noise Field=<0 0 1>
    Random Seed=10
   )
   &Drag(Drag=2.8, Use Linear Drag=true, Use Rotational Drag=true)
   &ScaleVelocity(
    Coordinate Space:ENiagaraCoordinateSpace=::Simulation
    Velocity Scale=&VectorFromFloat(Value=&FloatFromCurve(
      FloatCurve=DataInterface() {propertyValues {Curve {keys [
          {arriveTangent=-0.1413, interpMode=RCIM_Cubic, time=0, value=0.8}
          {arriveTangent=-0.1274, interpMode=RCIM_Cubic, leaveTangent=-0.1274, tangentMode=RCTM_User, time=0.4, value=0.2}
          <at 0.7033 0.15 RCIM_Cubic>
         ]}}}
      Scale Curve=&Multiply_Float(A=3, B=$User.VelocityMultiplier)
     ))
   )
   &SolveForcesAndVelocity(Clamp Velocity=false, Limit Acceleration=false)
   &ScaleSpriteSizeBySpeed(Max Scale Factor=<1 3>)
   &ScaleColor(Scale Alpha=$System.AgeFade, ScaleA=true, ScaleRGB=true) {module=ScaleColor001}
  ]
  rendererClasses [&NiagaraSpriteRendererProperties]
  renderers [
   {rendererClass=&NiagaraSpriteRendererProperties}
   {rendererClass=&NiagaraSpriteRendererProperties, rendererIndex=1}
  ]
 }
 FiresprintSubUV {
  emitterUpdateScript [
   &EmitterState(Life Cycle Mode=::NewEnumerator1, Loop Behavior=::NewEnumerator1, Loop Duration=0.9, UseLoopDelay=false)
   &SpawnRate(
    SpawnRate=&FloatFromCurve(
     FloatCurve=DataInterface() {propertyValues {Curve {keys [<at 0.6 1 RCIM_Cubic>,<at 1 0>]}}}
     CurveIndex=$System.Age
     Scale Curve=&Lerp_Float(A=10, B=18, Alpha=$System.HotdogDist)
    )
    Use Spawn Probability=false
   ) {module=SpawnRate001}
   &SpawnRate(
    SpawnGroup=1
    SpawnRate=&FloatFromCurve(
     FloatCurve=DataInterface() {propertyValues {Curve {keys [<at 0.6 1 RCIM_Cubic>,<at 0.94 0>]}}}
     CurveIndex=$System.Age
     Scale Curve=&Lerp_Float(A=15, B=30, Alpha=$System.HotdogDist)
    )
    Use Spawn Probability=false
   )
  ]
  particleSpawnScript [
   /SetVariables(Particles.Random size:NiagaraFloat=&RandomRangeFloat(Minimum=0.15, Maximum=0.25)) {isSetParameters=true}
   &InitializeParticle(
    Color=&MergeHelper_ColorSwitch(
     0=&RandomRangeLinearColor(Minimum=<rgba 5 0.3483 0.0927 1>, Maximum=<rgba 5 1.1349 0.0756 1>)
     1=<rgba 1 1 1 1>
     2=?
     3=?
     Index=&SpawnGroup()
     Num=::NewEnumerator0
    )
    Lifetime=&MergeHelper_FloatSwitch(
     0=&Multiply_Float(A=&RandomRangeFloat(Minimum=0.6, Maximum=0.7), B=$User.LifeMultiplier)
     1=&Multiply_Float(A=&RandomRangeFloat(Minimum=0.6, Maximum=0.7), B=$User.LifeMultiplier)
     2=?
     3=?
     Index=&SpawnGroup()
     Num=::NewEnumerator0
    )
    Sprite Rotation Mode=::NewEnumerator2
    Sprite Size=&MergeHelper_Vector2Switch(
     0=<10 10>
     1=&Multiply_Vector2D_ByFloat(Float=&Multiply_Float(A=70, B=$Particles.Random size), Vector2D=<1 1>)
     2=?
     3=?
     4=?
     Index=&SpawnGroup()
     Num=::NewEnumerator0
    )
    Sprite Size Mode=::NewEnumerator4
    UsePositionOffset=true
    Write Material Random=true
   )
   &CameraOffset(Camera Offset Amount=3)
   /SetVariables(Particles.AORandFloat:NiagaraFloat=&UniformAOrBFloat(A=0, Distribution Weight=0.3, Evaluation Type=::NewEnumerator0, Recalculate Random Each Loop=true, Override Seed=false)) {isSetParameters=true}
   &AddVelocity(
    Cone Angle=&Multiply_Float(
     A=$System.VelocityConeAngle
     B=&MergeHelper_FloatSwitch(0=0.8, 1=0.6, 2=?, 3=?, Index=&SpawnGroup(), Num=::NewEnumerator0)
    )
    Cone Axis=$System.VelocityConeAxis
    Module.UseConeFalloff=false
    Module.UseVelDistribution=false
    Override Seed=false
    Velocity Mode=::NewEnumerator2
    Velocity Speed=&Multiply_Float(
     A=&RandomRangeFloat(Minimum=$System.VelocitySpeedMin, Maximum=$System.VelocitySpeedMax)
     B=&MergeHelper_FloatSwitch(0=0.8, 1=1, 2=?, 3=?, Index=&SpawnGroup(), Num=::NewEnumerator0)
    )
   ) {module=AddVelocity001}
   /SetVariables(Particles.FireSpritesDynamic:Vector4f=<0 0 0 2.5>) {isSetParameters=true}
   /SetVariables(Particles.VisibilityTag=&SpawnGroup()) {isSetParameters=true}
  ]
  particleUpdateScript [
   &ParticleState()
   &ScaleSpriteSize(
    Scale Sprite Size Mode=::NewEnumerator0
    Uniform Scale Factor=&MergeHelper_FloatSwitch(
     0=&FloatFromCurve(FloatCurve=DataInterface() {propertyValues {Curve {keys [<at 0 0.5 RCIM_Cubic>,<at 1 1.5>]}}}, Scale Curve=6)
     1=&FloatFromCurve(
      FloatCurve=DataInterface() {propertyValues {Curve {keys [
          {arriveTangent=0.7245, interpMode=RCIM_Cubic, leaveTangent=0.7245, tangentMode=RCTM_User, time=0, value=0.2026}
          {interpMode=RCIM_Cubic, tangentMode=RCTM_User, time=0.377, value=0.3}
          {tangentMode=RCTM_User, time=1, value=0.1}
         ]}}}
      Scale Curve=11
     )
     2=?
     3=?
     Index=&SpawnGroup()
     Num=::NewEnumerator0
    )
   )
   &SpriteRotationRate(Rotation Rate=&RandomRangeFloat(Minimum=-100, Maximum=100))
   &ScaleColor(
    Scale Alpha=&MergeHelper_FloatSwitch(
     0=&Multiply_Float(
      A=&RandomRangeFloat(Minimum=1, Maximum=1)
      B=&FloatFromCurve(
       FloatCurve=DataInterface() {propertyValues {Curve {keys [
           {arriveTangent=10.7127, interpMode=RCIM_Cubic, leaveTangent=10.7127, tangentMode=RCTM_User, time=0, value=0}
           {interpMode=RCIM_Cubic, tangentMode=RCTM_User, time=0.0992, value=1}
           {arriveTangent=-1.1101, interpMode=RCIM_Cubic, leaveTangent=-1.1101, time=0.7122, value=0.8935}
           {arriveTangent=-2.7944, interpMode=RCIM_Cubic, leaveTangent=-2.7944, tangentMode=RCTM_User, time=1, value=0}
          ]}}}
       Scale Curve=0.2
      )
     )
     1=&Multiply_Float(
      A=2
      B=&FloatFromCurve(
       FloatCurve=DataInterface() {propertyValues {Curve {keys [
           {arriveTangent=3.6955, interpMode=RCIM_Cubic, leaveTangent=3.6955, tangentMode=RCTM_User, time=0, value=0}
           {arriveTangent=1.1708, interpMode=RCIM_Cubic, leaveTangent=1.1708, time=0.3536, value=1}
           {arriveTangent=-1.547, interpMode=RCIM_Cubic, leaveTangent=-1.547, time=0.8541, value=1}
           {tangentMode=RCTM_User, time=1, value=0}
          ]}}}
       Scale Curve=&Multiply_Float(
        A=&FN_FrontEndFloatModifier(Gameplay=0.8, Front End=0.6)
        B=&Lerp_Float(A=0.22, B=0.6, Alpha=$Particles.AORandFloat)
       )
      )
     )
     2=?
     3=?
     Index=&SpawnGroup()
     Num=::NewEnumerator0
    )
    Scale RGB=&MergeHelper_VectorSwitch(
     0=&Multiply_VectorByFloat(Vector=<1 1 1>, Float=1)
     1=&Multiply_VectorByFloat(
      Vector=&MakeVectorFromLinearColorRGB(Color=&LerpLinearColors(
        StartColor=&ColorFromCurve(ColorCurve=DataInterface() {propertyValues {
           AlphaCurve {keys []}
           BlueCurve {keys [<at 0.2645 0.474>,<at 1 0.3855>]}
           GreenCurve {keys [<at 0.2645 3.2001>,<at 1 1.825>]}
           RedCurve {keys [<at 0.2645 13>,<at 1 7>]}
          }})
        EndColor=<rgba 0.0938 0.0138 0.003 1>
        LerpFactor=$Particles.AORandFloat
       ))
      Float=1
     )
     2=?
     3=?
     Index=&SpawnGroup()
     Num=::NewEnumerator0
    )
    ScaleA=true
    ScaleRGB=true
   )
   &SubUVAnimation(
    End Frame Range Override:NiagaraInt32=&RandomRangeInt(Minimum=48, Maximum=60)
    "Sprite Renderer "=DataInterface() {propertyValues {SpriteRenderer=&self:FiresprintSubUV_1.NiagaraSpriteRendererProperties_0}}
    Start Frame Range Override=&RandomRangeInt(Minimum=15, Maximum=30)
    UseEndFrame=true
    UseStartFrame=true
   )
   &SubUVAnimation(
    End Frame Range Override=12
    "Sprite Renderer "=DataInterface() {propertyValues {SpriteRenderer=&self:FiresprintSubUV_1.NiagaraSpriteRendererProperties_1}}
    Start Frame Range Override=8
    SubUV Lookup Index Scale:NiagaraFloat=&RandomRangeFloat(Minimum=0.2, Maximum=1)
    UseEndFrame=true
    UseStartFrame=true
   ) {module=SubUVAnimation001}
   &CameraOffset(Camera Offset Amount=15) {module=CameraOffset001}
   &Drag(
    Drag=&MergeHelper_FloatSwitch(0=2, 1=2.7, 2=?, 3=?, Index=&SpawnGroup(), Num=::NewEnumerator0)
    Use Linear Drag=true
    Use Rotational Drag=true
   )
   &CurlNoiseForce(
    Bias Noise=false
    Noise Frequency=15
    Noise Strength=&MergeHelper_FloatSwitch(
     0=&FloatFromCurve(FloatCurve=DataInterface() {propertyValues {Curve {keys [<at 0 0 RCIM_Cubic>,<at 1 1>]}}}, Scale Curve=300)
     1=&Multiply_Float(
      A=1500
      B=&FloatFromCurve(FloatCurve=DataInterface() {propertyValues {Curve {keys [{arriveTangent=2.1957, interpMode=RCIM_Cubic, leaveTangent=2.1957, tangentMode=RCTM_User, time=0, value=0}, <at 1 1>]}}})
     )
     2=?
     3=?
     Index=&SpawnGroup()
     Num=::NewEnumerator0
    )
    Pan Noise=true
    Pan Noise Field=<0 0 1>
   )
   &ScaleVelocity(
    Coordinate Space:ENiagaraCoordinateSpace=::Simulation
    Velocity Scale=&VectorFromFloat(Value=&FloatFromCurve(
      FloatCurve=DataInterface() {propertyValues {Curve {keys [
          {arriveTangent=-0.1413, interpMode=RCIM_Cubic, time=0, value=0.8}
          {arriveTangent=-0.1274, interpMode=RCIM_Cubic, leaveTangent=-0.1274, tangentMode=RCTM_User, time=0.4, value=0.2}
          <at 0.7033 0.15 RCIM_Cubic>
         ]}}}
      Scale Curve=&Multiply_Float(A=3, B=$User.VelocityMultiplier)
     ))
   )
   &SolveForcesAndVelocity(Clamp Velocity=false, Limit Acceleration=false)
   &ScaleColor(Scale Alpha=$System.AgeFade, ScaleA=true, ScaleRGB=true) {module=ScaleColor002}
  ]
  rendererClasses [&NiagaraSpriteRendererProperties]
  renderers [
   {rendererClass=&NiagaraSpriteRendererProperties}
   {rendererClass=&NiagaraSpriteRendererProperties, rendererIndex=1}
  ]
 }
 Glow {
  emitterUpdateScript [
   &EmitterState(Life Cycle Mode=::NewEnumerator1, Loop Behavior=::NewEnumerator1, Loop Duration=1, UseLoopDelay=false)
   &SpawnBurst_Instantaneous(Use Spawn Probability=false, UseLoopCountLimit=false)
   &SpawnBurst_Instantaneous(Spawn Group=1, Use Spawn Probability=false, UseLoopCountLimit=false) {module=SpawnBurst_Instantaneous001}
   &SpawnBurst_Instantaneous(Spawn Count=3, Spawn Group=2, Use Spawn Probability=false, UseLoopCountLimit=false) {module=SpawnBurst_Instantaneous002}
   &SpawnBurst_Instantaneous(Spawn Count=2, Spawn Group=2, Spawn Time=0.05, Use Spawn Probability=false, UseLoopCountLimit=false) {module=SpawnBurst_Instantaneous003}
  ]
  particleSpawnScript [
   &InitializeParticle(
    Color=&MergeHelper_ColorSwitch(
     0=<rgba 0.4809 0.4809 0.4809 3>
     1=<rgba 0.4809 0.4809 0.4809 3>
     2=<rgba 2 2 2 1>
     3=?
     Index=&SpawnGroup()
     Num=::NewEnumerator1
    )
    Lifetime=&MergeHelper_FloatSwitch(
     0=1.05
     1=1.05
     2=&Multiply_Float(A=&RandomRangeFloat(Minimum=0.2, Maximum=0.3), B=$User.LifeMultiplier)
     3=&Multiply_Float(A=&RandomRangeFloat(Minimum=0.2, Maximum=0.3), B=$User.LifeMultiplier)
     Index=&SpawnGroup()
     Num=::NewEnumerator1
    )
    Mesh Scale=&Multiply_VectorByFloat(
     Vector=&MakeVector(X=0.925, Y=0.925, Z=&RandomRangeFloat(Minimum=2, Maximum=3))
     Float=&RandomRangeFloat(Minimum=0.3, Maximum=0.4)
    )
    Mesh Scale Mode=::NewEnumerator4
    Position Offset=<0 0 4>
    Sprite Size Mode=::NewEnumerator3
    Uniform Sprite Size=160
    UsePositionOffset=true
    Write Material Random=true
   )
   &InitialMeshOrientation(
    Mesh Axis To Orient=<0 0 1>
    Mesh Orientation Mode=::NewEnumerator2
    Orientation Vector=<1 0 -0.1408>
    Rotation=&Add_Vector(A=&RandomRangeVector(Minimum=<0 0 0>, Maximum=<0 0 1>), B=<0 0 0>)
    Use Rotation Vector=true
   ) {module=InitialMeshOrientation001}
   &CameraOffset(Camera Offset Amount=2)
   /SetVariables(Particles.VisibilityTag=&SpawnGroup()) {isSetParameters=true}
   /SetVariables(Particles.AORandFloat:NiagaraFloat=&UniformAOrBFloat(A=0, Distribution Weight=0.3, Evaluation Type=::NewEnumerator0, Recalculate Random Each Loop=true, Override Seed=false)) {isSetParameters=true}
   &AddVelocity(
    Cone Angle=$System.VelocityConeAngle
    Cone Axis=$System.VelocityConeAxis
    Module.UseConeFalloff=false
    Module.UseVelDistribution=false
    Override Seed=false
    Velocity Mode=::NewEnumerator2
    Velocity Speed=&MergeHelper_FloatSwitch(
     0=0
     1=0
     2=&Multiply_Float(A=&RandomRangeFloat(Minimum=$System.VelocitySpeedMin, Maximum=$System.VelocitySpeedMax), B=0.7)
     3=?
     Index=&SpawnGroup()
     Num=::NewEnumerator1
    )
   ) {module=AddVelocity001}
  ]
  particleUpdateScript [
   /SetVariables(Particles.SizeScale:Vector3f=&Multiply_VectorByFloat(Vector=<1 1 0.5>, Float=5)) {isSetParameters=true}
   &Drag(Drag=2, Use Linear Drag=true, Use Rotational Drag=false)
   &SolveForcesAndVelocity(Clamp Velocity=false, Limit Acceleration=false)
   &ParticleState()
   &DynamicMaterialParameters(
    Index 0 Param 1=3.5
    Index 0 Param 2=10
    Index 0 Param 4=&FloatFromCurve(FloatCurve=DataInterface() {propertyValues {Curve {keys [<at 0 0.8 RCIM_Cubic>,<at 1 0.3>]}}})
    Module.Index0Float0Write=true
    Module.Index0Float1Write=true
    Module.Index0Float2Write=true
    Module.Index0Float3Write=true
   )
   &Light_Attributes(
    Color=1
    Enabled=1
    Falloff Exponent=1
    Light Exponent=2
    Light Position Offset:Vector3f=&Lerp_Vector(A=<-30 0 -20>, B=<-70 0 -20>, Alpha=$System.HotdogDist)
    Light Radius=120
    Light Specular Scale=0.5
    Position=1
    Position Offset Coordinate Space:ENiagaraCoordinateSpace=::Local
    Radius=1
    Scale RGB:Vector3f=&MakeVectorFromLinearColorRGB(Color=&Multiply_LinearColorByFloat(
      Linear Color=<rgba 50 14.2448 2.88 1>
      Float=&FloatFromCurve(
       FloatCurve=DataInterface() {propertyValues {Curve {keys [<at 0 0 RCIM_Cubic>, <at 0.1354 1 RCIM_Cubic>, {arriveTangent=-0.8444, tangentMode=RCTM_User, time=1, value=0}]}}}
       Scale Curve=11
      )
     ))
    ScaleA=true
    ScaleRGB=true
    Specular Scale=1
    UsePositionOffset=true
   )
   &ScaleColor(
    Scale Alpha=&MergeHelper_FloatSwitch(
     0=&FloatFromCurve(
      FloatCurve=DataInterface() {propertyValues {Curve {keys [
          <at 0 0 RCIM_Cubic>
          <at 0.1061 1 RCIM_Cubic>
          {arriveTangent=-1.098, interpMode=RCIM_Cubic, leaveTangent=-1.098, tangentMode=RCTM_User, time=1, value=0}
         ]}}}
      Scale Curve=0.1
     )
     1=&FloatFromCurve(
      FloatCurve=DataInterface() {propertyValues {Curve {keys [
          <at 0 0 RCIM_Cubic>
          <at 0.1061 1 RCIM_Cubic>
          {arriveTangent=-1.098, interpMode=RCIM_Cubic, leaveTangent=-1.098, tangentMode=RCTM_User, time=1, value=0}
         ]}}}
      Scale Curve=0.1
     )
     2=&FloatFromCurve(
      FloatCurve=DataInterface() {propertyValues {Curve {keys [
          {arriveTangent=-0.0329, interpMode=RCIM_Cubic, leaveTangent=-0.0329, tangentMode=RCTM_User, time=0, value=0}
          {arriveTangent=3.3381, interpMode=RCIM_Cubic, leaveTangent=3.3381, tangentMode=RCTM_User, time=0.1262, value=0.6}
          {arriveTangent=0.0085, interpMode=RCIM_Cubic, leaveTangent=0.0085, tangentMode=RCTM_User, time=0.2, value=0.75}
          {arriveTangent=0.0414, interpMode=RCIM_Cubic, leaveTangent=0.0414, tangentMode=RCTM_User, time=1, value=0}
         ]}}}
      Scale Curve=&Multiply_Float(
       A=&RandomRangeFloat(Minimum=1, Maximum=2)
       B=&Lerp_Float(A=1, B=2, Alpha=$Particles.AORandFloat)
      )
     )
     3=?
     Index=&SpawnGroup()
     Num=::NewEnumerator1
    )
    Scale RGB=&MergeHelper_VectorSwitch(
     0=&MakeVectorFromLinearColorRGB(Color=<rgba 2 0.6784 0.1045 1>)
     1=&MakeVectorFromLinearColorRGB(Color=<rgba 2 0.6784 0.1045 1>)
     2=&MakeVectorFromLinearColorRGB(Color=&LerpLinearColors(
       StartColor=&RandomRangeLinearColor(Minimum=<rgba 2 0.7197 0 1>, Maximum=<rgba 2 0.134 0 1>)
       EndColor=&RandomRangeLinearColor(Minimum=<rgba 0.1719 0.0647 0.0045 1>, Maximum=<rgba 0.0938 0.009 0.0029 1>)
       LerpFactor=$Particles.AORandFloat
      ))
     3=?
     Index=&SpawnGroup()
     Num=::NewEnumerator1
    )
    ScaleA=true
    ScaleRGB=true
   )
   &FN_Sprite_Attributes(
    Camera Offset=15
    Color=&MakeLinearColorFromVectorAndFloat(
     "Vector (RGB)"=&Multiply_Vector(A=&MakeVectorFromLinearColorRGB(Color=<rgba 25 4.2556 0.0197 1>), B=<1 1 1>)
     "Float (Alpha)"=&Multiply_Float(
      A=&FloatFromCurve(
       FloatCurve=DataInterface() {propertyValues {Curve {keys [
           {interpMode=RCIM_Cubic, tangentMode=RCTM_User, time=0, value=0}
           {interpMode=RCIM_Cubic, tangentMode=RCTM_User, time=0.3, value=1}
           {tangentMode=RCTM_User, time=1, value=0}
          ]}}}
       Scale Curve=0.7
      )
      B=1
     )
    )
    Create Camera Offset Attribute=true
    Create Color Attribute=true
    Create SpriteSize Attribute=true
    SpriteSize=<25 25>
    Write to Module Namespace=true
   )
   &ScaleMeshSize(Scale Factor=&Multiply_VectorByFloat(
     Vector=&Multiply_VectorByFloat(
      Vector=&MakeVector(
       X=1
       Y=1
       Z=&Multiply_Float(
        A=&FloatFromCurve(FloatCurve=DataInterface() {propertyValues {Curve {keys [{arriveTangent=-1.572, interpMode=RCIM_Cubic, time=0, value=1}, {interpMode=RCIM_Cubic, tangentMode=RCTM_User, time=1, value=0.6791}]}}})
        B=&RandomRangeFloat(Minimum=1, Maximum=1)
       )
      )
      Float=1.2
     )
     Float=&FloatFromCurve(
      FloatCurve=DataInterface() {propertyValues {Curve {keys [
          {arriveTangent=3.0662, interpMode=RCIM_Cubic, leaveTangent=3.0662, tangentMode=RCTM_User, time=0, value=0}
          {interpMode=RCIM_Cubic, leaveTangent=0.9183, time=1, value=3.8934}
         ]}}}
      Scale Curve=&Multiply_Float(A=0.5, B=0.7)
     )
    ))
   /SetVariables(Particles.MeshRingDynamic:Vector4f=&MakeVector4(
     X=&FloatFromCurve(
      FloatCurve=DataInterface() {propertyValues {Curve {keys [{interpMode=RCIM_Cubic, leaveTangent=-0.0324, tangentMode=RCTM_Break, time=0, value=1.5}, <at 1 -0.5>]}}}
      Scale Curve=&RandomRangeFloat(Minimum=0.8, Maximum=1)
     )
     Y=&FloatFromCurve(FloatCurve=DataInterface() {propertyValues {Curve {keys [
          {arriveTangent=0.3262, interpMode=RCIM_Cubic, leaveTangent=0.3262, tangentMode=RCTM_User, time=0.2, value=0}
          {arriveTangent=1.6237, interpMode=RCIM_Cubic, leaveTangent=1.6237, tangentMode=RCTM_User, time=1, value=1}
         ]}}})
     Z=0
     W=0
    )) {isSetParameters=true}
  ]
  rendererClasses [
   &NiagaraSpriteRendererProperties
   &NiagaraLightRendererProperties
   &NiagaraMeshRendererProperties
  ]
  renderers [
   {rendererClass=&NiagaraSpriteRendererProperties}
   {rendererClass=&NiagaraLightRendererProperties, rendererIndex=1}
   {rendererClass=&NiagaraSpriteRendererProperties, rendererIndex=2}
   {rendererClass=&NiagaraMeshRendererProperties, rendererIndex=3}
  ]
 }
 MainFire {
  emitterUpdateScript [
   &EmitterState(Life Cycle Mode=::NewEnumerator1, Loop Behavior=::NewEnumerator1, Loop Duration=0.8, UseLoopDelay=true)
   &SpawnRate(
    SpawnRate=&FloatFromCurve(
     FloatCurve=DataInterface() {propertyValues {Curve {keys [<at 0.6 1 RCIM_Cubic>,<at 1 0>]}}}
     CurveIndex=$System.Age
     Scale Curve=&Lerp_Float(A=50, B=85, Alpha=$System.HotdogDist)
    )
    Use Spawn Probability=false
   )
   &SpawnRate(
    SpawnGroup=1
    SpawnRate=&FloatFromCurve(
     FloatCurve=DataInterface() {propertyValues {Curve {keys [<at 0.6 1 RCIM_Cubic>,<at 1 0>]}}}
     CurveIndex=$System.Age
     Scale Curve=&Lerp_Float(A=20, B=30, Alpha=$System.HotdogDist)
    )
    Use Spawn Probability=false
   ) {module=SpawnRate001}
  ]
  particleSpawnScript [
   /SetVariables(Particles.Random size:NiagaraFloat=&RandomRangeFloat(Minimum=0.4, Maximum=0.6)) {isSetParameters=true}
   &InitializeParticle(
    Color=&RandomRangeLinearColor(Minimum=<rgba 0.4896 0.4896 0.4896 1>, Maximum=<rgba 1 1 1 1>)
    Lifetime=&MergeHelper_FloatSwitch(
     0=&Multiply_Float(A=&RandomRangeFloat(Minimum=0.4, Maximum=0.5), B=$User.LifeMultiplier)
     1=&Multiply_Float(A=&RandomRangeFloat(Minimum=0.6, Maximum=0.7), B=$User.LifeMultiplier)
     2=?
     3=?
     Index=&SpawnGroup()
     Num=::NewEnumerator0
    )
    Sprite Rotation Mode=::NewEnumerator2
    Sprite Size Mode=::NewEnumerator3
    Uniform Sprite Size=&Multiply_Float(A=60, B=$Particles.Random size)
    UsePositionOffset=false
    Write Material Random=true
   )
   &DynamicMaterialParameters(
    Index 0 Param 4=2.5
    Module.Param0WriteEnabled=true
    Module.Param1WriteEnabled=true
    Module.Param2WriteEnabled=false
    Module.Param3WriteEnabled=true
   ) {module=DynamicMaterialParameters001}
   &AddVelocity(
    Cone Angle=&Multiply_Float(
     A=$System.VelocityConeAngle
     B=&MergeHelper_FloatSwitch(0=0.5, 1=0.9, 2=?, 3=?, Index=&SpawnGroup(), Num=::NewEnumerator0)
    )
    Cone Axis=$System.VelocityConeAxis
    Module.UseConeFalloff=false
    Module.UseVelDistribution=false
    Override Seed=false
    Velocity Mode=::NewEnumerator2
    Velocity Speed=&MergeHelper_FloatSwitch(
     0=&Multiply_Float(A=1.3, B=&RandomRangeFloat(Minimum=$System.VelocitySpeedMin, Maximum=$System.VelocitySpeedMin))
     1=&Multiply_Float(A=&RandomRangeFloat(Minimum=$System.VelocitySpeedMin, Maximum=$System.VelocitySpeedMax), B=0.8)
     2=?
     3=?
     Index=&SpawnGroup()
     Num=::NewEnumerator0
    )
   ) {module=AddVelocity001}
  ]
  particleUpdateScript [
   &ParticleState()
   /SetVariables(Particles.Size Scale:NiagaraFloat=&MergeHelper_FloatSwitch(
     0=&FloatFromCurve(
      FloatCurve=DataInterface() {propertyValues {Curve {keys [
          {arriveTangent=0.8395, interpMode=RCIM_Cubic, leaveTangent=0.8395, tangentMode=RCTM_User, time=0.05, value=0.13}
          {tangentMode=RCTM_User, time=1, value=0.4}
         ]}}}
      Scale Curve=2.5
     )
     1=&FloatFromCurve(
      FloatCurve=DataInterface() {propertyValues {Curve {keys [
          {arriveTangent=1.1111, interpMode=RCIM_Cubic, leaveTangent=1.1111, tangentMode=RCTM_User, time=0, value=0.5365}
          {tangentMode=RCTM_User, time=1, value=1}
         ]}}}
      Scale Curve=1.8
     )
     2=?
     3=?
     Index=&SpawnGroup()
     Num=::NewEnumerator0
    )) {isSetParameters=true}
   &Drag(Drag=3, Use Linear Drag=true, Use Rotational Drag=true)
   &CurlNoiseForce(
    Bias Noise=false
    Noise Frequency=15
    Noise Strength=&FloatFromCurve(
     FloatCurve=DataInterface() {propertyValues {Curve {keys [<at 0 0 RCIM_Cubic>,<at 1 1>]}}}
     Scale Curve=&MergeHelper_FloatSwitch(0=0, 1=1500, 2=?, 3=?, Index=&SpawnGroup(), Num=::NewEnumerator0)
    )
    Pan Noise=true
    Pan Noise Field=<0 0 1>
   )
   &ScaleVelocity(
    Coordinate Space:ENiagaraCoordinateSpace=::Simulation
    Velocity Scale=&MergeHelper_VectorSwitch(
     0=<1 1 1>
     1=&VectorFromFloat(Value=&FloatFromCurve(
       FloatCurve=DataInterface() {propertyValues {Curve {keys [
           {arriveTangent=-0.1413, interpMode=RCIM_Cubic, time=0, value=0.8}
           {arriveTangent=-0.1274, interpMode=RCIM_Cubic, leaveTangent=-0.1274, tangentMode=RCTM_User, time=0.4, value=0.2}
           <at 0.7033 0.15 RCIM_Cubic>
          ]}}}
       Scale Curve=&Multiply_Float(A=3, B=$User.VelocityMultiplier)
      ))
     2=?
     3=?
     Index=&SpawnGroup()
     Num=::NewEnumerator0
    )
   )
   &SolveForcesAndVelocity(Clamp Velocity=false, Limit Acceleration=false)
   &SpriteRotationRate(Rotation Rate=&RandomRangeFloat(Minimum=-100, Maximum=100)) {module=SpriteRotationRate001}
   &ScaleSpriteSize(
    Scale Factor=&Multiply_Vector2D_ByFloat(
     Float=$Particles.Size Scale
     Vector2D=&MergeHelper_Vector2Switch(
      0=&Vector2DFromCurve(
       Vector2Curve=DataInterface() {propertyValues {XCurve {keys [<at 0 1>,<at 1 1>]}, YCurve {keys [<at 0 3 RCIM_Cubic>,<at 1 1.4>]}}}
       ScaleCurve=<1 1>
      )
      1=<1 1>
      2=?
      3=?
      4=?
      Index=&SpawnGroup()
      Num=::NewEnumerator0
     )
    )
    Scale Sprite Size Mode=::NewEnumerator2
   ) {module=ScaleSpriteSize001}
   &ScaleColor(
    Scale Alpha=&MergeHelper_FloatSwitch(
     0=&Multiply_Float(
      A=2
      B=&FloatFromCurve(
       FloatCurve=DataInterface() {propertyValues {Curve {keys [<at 0 1 RCIM_Cubic>, {interpMode=RCIM_Cubic, tangentMode=RCTM_User, time=0.8, value=1}, {tangentMode=RCTM_User, time=0.8719, value=0}]}}}
       Scale Curve=0.7
      )
     )
     1=&Multiply_Float(
      A=0.848
      B=&FloatFromCurve(
       FloatCurve=DataInterface() {propertyValues {Curve {keys [<at 0 1 RCIM_Cubic>, {interpMode=RCIM_Cubic, tangentMode=RCTM_User, time=0.4, value=1}, {tangentMode=RCTM_User, time=1, value=0}]}}}
       Scale Curve=0.5
      )
     )
     2=?
     3=?
     Index=&SpawnGroup()
     Num=::NewEnumerator0
    )
    Scale RGB=&Multiply_VectorByFloat(
     Vector=&MergeHelper_VectorSwitch(
      0=&Multiply_VectorByFloat(
       Vector=&MakeVectorFromLinearColorRGB(Color=&ColorFromCurve(ColorCurve=DataInterface() {propertyValues {
           AlphaCurve {keys []}
           BlueCurve {keys [<at 0.004 0.1917>,<at 0.2948 0.1146>,<at 0.6421 0.1036>,<at 1 0.0451 RCIM_Cubic>]}
           GreenCurve {keys [<at 0.004 0.7162>,<at 0.2948 0.8168>,<at 0.6421 0.7177>,<at 1 0.1894 RCIM_Cubic>]}
           RedCurve {keys [<at 0.004 1.6>,<at 0.2948 2>,<at 0.6421 1.8419>,<at 1 1 RCIM_Cubic>]}
          }}))
       Float=&FloatFromCurve(
        FloatCurve=DataInterface() {propertyValues {Curve {keys [
            <at 0.0314 5>
            {arriveTangent=-10.8788, interpMode=RCIM_Cubic, leaveTangent=-10.8788, time=0.1346, value=3.4908}
            {arriveTangent=-3.3406, interpMode=RCIM_Cubic, leaveTangent=-3.3406, time=0.3531, value=1.5}
            <at 1 0.6 RCIM_Cubic>
           ]}}}
        Scale Curve=2
       )
      )
      1=&Multiply_VectorByFloat(
       Vector=&MakeVectorFromLinearColorRGB(Color=&LerpLinearColors(
         StartColor=&ColorFromCurve(ColorCurve=DataInterface() {propertyValues {
            AlphaCurve {keys []}
            BlueCurve {keys [<at 0.0268 0.0708>,<at 0.7901 0.0242 RCIM_Cubic>]}
            GreenCurve {keys [<at 0.0268 0.6333>,<at 0.7901 0.501 RCIM_Cubic>]}
            RedCurve {keys [<at 0.0268 2>,<at 0.7901 2 RCIM_Cubic>]}
           }})
         EndColor=&ColorFromCurve(ColorCurve=DataInterface() {propertyValues {
            AlphaCurve {keys []}
            BlueCurve {keys [<at 0.0268 0.0596>,<at 0.7901 0.0773 RCIM_Cubic>]}
            GreenCurve {keys [<at 0.0268 0.0978>,<at 0.7901 0.1738 RCIM_Cubic>]}
            RedCurve {keys [<at 0.0268 2>,<at 0.7901 2 RCIM_Cubic>]}
           }})
         LerpFactor=&RandomRangeFloat(Minimum=0, Maximum=1)
        ))
       Float=&FloatFromCurve(FloatCurve=DataInterface() {propertyValues {Curve {keys [<at 0.3 2.25>,<at 1 0.1>]}}})
      )
      2=?
      3=?
      Index=&SpawnGroup()
      Num=::NewEnumerator0
     )
     Float=&Lerp_Float(A=1, B=0.7, Alpha=$System.FrontEndLerp)
    )
    ScaleA=true
    ScaleRGB=true
   )
   &FN_Sprite_Attributes(
    Color=&MergeHelper_ColorSwitch(
     0=&MakeLinearColorFromVectorAndFloat(
      "Vector (RGB)"=&MakeVectorFromLinearColorRGB(Color=<rgba 7 3.2439 1.0573 0.5>)
      "Float (Alpha)"=&Multiply_Float(
       A=&FloatFromCurve(
        FloatCurve=DataInterface() {propertyValues {Curve {keys [<at 0 0.5428 RCIM_Cubic>, {arriveTangent=-0.5428, interpMode=RCIM_Cubic, leaveTangent=-0.5428, time=0.1667, value=1}, <at 1 0 RCIM_Cubic>]}}}
        Scale Curve=0.512
       )
       B=&FN_FrontEndFloatModifier(Gameplay=0.3, Front End=0.1)
      )
     )
     1=&MakeLinearColorFromVectorAndFloat(
      "Vector (RGB)"=&MakeVectorFromLinearColorRGB(Color=&RandomRangeLinearColor(Minimum=<rgba 1 0.3754 0.087 0.5>, Maximum=<rgba 1 0.1603 0.0646 0.5>))
      "Float (Alpha)"=&Multiply_Float(
       A=&FloatFromCurve(
        FloatCurve=DataInterface() {propertyValues {Curve {keys [
            <at 0 0 RCIM_Cubic>
            <at 0.1334 1 RCIM_Cubic>
            {arriveTangent=-1.9305, interpMode=RCIM_Cubic, leaveTangent=-1.9305, tangentMode=RCTM_User, time=1, value=0}
           ]}}}
        Scale Curve=0.52
       )
       B=&Multiply_Float(A=&FN_FrontEndFloatModifier(Gameplay=0.3, Front End=0.1), B=$System.AgeFade)
      )
     )
     2=?
     3=?
     Index=&SpawnGroup()
     Num=::NewEnumerator0
    )
    Create Color Attribute=true
    Create SpriteSize Attribute=true
    SpriteSize=&MergeHelper_Vector2Switch(
     0=&Multiply_Vector2D(
      A=&Vector2DFromFloat(Value=&Multiply_Float(A=92.8817, B=&Multiply_Float(A=$Particles.Random size, B=$Particles.Size Scale)))
      B=<1 3>
     )
     1=&Vector2DFromFloat(Value=&Multiply_Float(A=90, B=&Multiply_Float(A=$Particles.Random size, B=$Particles.Size Scale)))
     2=?
     3=?
     4=?
     Index=&SpawnGroup()
     Num=::NewEnumerator0
    )
    Write to Module Namespace=true
   ) {module=FN_Sprite_Attributes001}
   &SubUVAnimation(
    End Frame Range Override=12
    Evaluation Type=::NewEnumerator0
    "Sprite Renderer "=DataInterface() {propertyValues {SpriteRenderer=&self:MainFire_0.NiagaraSpriteRendererProperties_0}}
    Start Frame Range Override=8
    SubUV Animation Mode:ENiagara_SubUVLookupModeV2=::NewEnumerator2
    UseEndFrame=false
    UseStartFrame=false
   )
   &ScaleSpriteSizeBySpeed(
    Max Scale Factor=&MergeHelper_Vector2Switch(0=<1.3 2>, 1=<1 2.2>, 2=?, 3=?, 4=?, Index=&SpawnGroup(), Num=::NewEnumerator0)
    Min Scale Factor=&MergeHelper_Vector2Switch(0=<1.3 1>, 1=<1 1>, 2=?, 3=?, 4=?, Index=&SpawnGroup(), Num=::NewEnumerator0)
   )
   &ScaleColor(Scale Alpha=$System.AgeFade, ScaleA=true, ScaleRGB=true) {module=ScaleColor001}
  ]
  rendererClasses [&NiagaraSpriteRendererProperties]
  renderers [
   {rendererClass=&NiagaraSpriteRendererProperties}
   {rendererClass=&NiagaraSpriteRendererProperties, rendererIndex=1}
  ]
 }
 Smoke {
  emitterUpdateScript [
   &EmitterState(Life Cycle Mode=::NewEnumerator1, Loop Behavior=::NewEnumerator1, Loop Duration=1, UseLoopDelay=false)
   &SpawnRate(
    SpawnRate=&FloatFromCurve(
     FloatCurve=DataInterface() {propertyValues {Curve {keys [<at 0.6 1 RCIM_Cubic>,<at 1 0>]}}}
     CurveIndex=$System.Age
     Scale Curve=&Lerp_Float(A=10, B=20, Alpha=$System.HotdogDist)
    )
    Use Spawn Probability=false
   )
  ]
  particleSpawnScript [
   &InitializeParticle(
    Color=<rgba 2 1.7786 1.7708 1>
    Lifetime=&Multiply_Float(A=&RandomRangeFloat(Minimum=0.6, Maximum=0.9), B=$User.LifeMultiplier)
    Sprite Rotation Mode=::NewEnumerator2
    Sprite Size Mode=::NewEnumerator1
    Uniform Sprite Size Max=70
    Uniform Sprite Size Min=50
    UsePositionOffset=false
    Write Material Random=true
   )
   &CameraOffset(Camera Offset Amount=10)
   /SetVariables(Particles.RandomLerp:NiagaraFloat=&MakeCustomFloatFromBool(Bool=&RandomBool(), True Float=0, False Float=1)) {isSetParameters=true}
   /SetVariables(Particles.AORandFloat:NiagaraFloat=&UniformAOrBFloat(A=0, Distribution Weight=0.4, Evaluation Type=::NewEnumerator0, Recalculate Random Each Loop=true, Override Seed=false)) {isSetParameters=true}
   &AddVelocity(
    Cone Angle=&Multiply_Float(A=$System.VelocityConeAngle, B=0.4)
    Cone Axis=$System.VelocityConeAxis
    Module.UseConeFalloff=false
    Module.UseVelDistribution=false
    Override Seed=false
    Velocity Mode=::NewEnumerator2
    Velocity Speed=&Multiply_Float(
     A=&RandomRangeFloat(Minimum=0.9, Maximum=1)
     B=&RandomRangeFloat(Minimum=$System.VelocitySpeedMin, Maximum=$System.VelocitySpeedMax)
    )
   ) {module=AddVelocity001}
  ]
  particleUpdateScript [
   &ParticleState()
   &ScaleSpriteSize(
    Uniform Curve Scale=1.25
    Uniform Curve Sprite Scale=DataInterface() {propertyValues {Curve {keys [<at 0 0.3119 RCIM_Cubic>, {arriveTangent=1.2881, interpMode=RCIM_Cubic, leaveTangent=1.2881, time=0.3, value=0.775}, <at 1 1.6 RCIM_Cubic>]}}}
   )
   &Drag(Drag=2, Use Linear Drag=true, Use Rotational Drag=true)
   &ScaleVelocity(
    Coordinate Space:ENiagaraCoordinateSpace=::Simulation
    Velocity Scale=&VectorFromFloat(Value=&FloatFromCurve(
      FloatCurve=DataInterface() {propertyValues {Curve {keys [
          {arriveTangent=-0.1413, interpMode=RCIM_Cubic, time=0, value=0.8}
          {arriveTangent=-0.1274, interpMode=RCIM_Cubic, leaveTangent=-0.1274, tangentMode=RCTM_User, time=0.4, value=0.2}
          <at 0.7033 0.15 RCIM_Cubic>
         ]}}}
      Scale Curve=&Multiply_Float(A=3, B=$User.VelocityMultiplier)
     ))
   )
   &GravityForce(Gravity=&Multiply_VectorByFloat(
     Vector=<0 0 5473.0942>
     Float=&FloatFromCurve(FloatCurve=DataInterface() {propertyValues {Curve {keys [<at 0 0.3 RCIM_Cubic>,<at 1 1>]}}})
    ))
   &SolveForcesAndVelocity(Clamp Velocity=false, Limit Acceleration=false)
   &ScaleColor(
    Scale Alpha=&FloatFromCurve(
     FloatCurve=DataInterface() {propertyValues {Curve {keys [
         <at 0 0 RCIM_Cubic>
         {arriveTangent=1.6857, interpMode=RCIM_Cubic, leaveTangent=1.6857, time=0.3302, value=1}
         {arriveTangent=-1.4929, interpMode=RCIM_Cubic, leaveTangent=-1.4929, time=0.4892, value=0.8247}
         {arriveTangent=-1.1447, interpMode=RCIM_Cubic, leaveTangent=-1.1447, tangentMode=RCTM_User, time=1, value=0}
        ]}}}
     Scale Curve=0.8
    )
    Scale RGB=&Multiply_VectorByFloat(
     Vector=&MakeVectorFromLinearColorRGB(Color=&LerpLinearColors(
       StartColor=&ColorFromCurve(ColorCurve=DataInterface() {propertyValues {
          AlphaCurve {keys [<at 0 1 RCIM_Cubic>,<at 1 0 RCIM_Cubic>]}
          BlueCurve {keys [<at 0.8778 0.0016 RCIM_Cubic>]}
          GreenCurve {keys [<at 0.8778 0.0263 RCIM_Cubic>]}
          RedCurve {keys [<at 0.8778 0.1354 RCIM_Cubic>]}
         }})
       EndColor=<rgba 0.0156 0.0026 0 1>
       LerpFactor=$Particles.AORandFloat
      ))
     Float=&FloatFromCurve(FloatCurve=DataInterface() {propertyValues {Curve {keys [<at 0 0.657>,<at 1 0.0455>]}}}, Scale Curve=3.3)
    )
    ScaleA=true
    ScaleRGB=true
   )
   &DynamicMaterialParameters(
    Index 0 Param 1=&Add_Float(
     A=&RandomRangeFloat(Minimum=0, Maximum=1)
     B=&FloatFromCurve(
      FloatCurve=DataInterface() {propertyValues {Curve {keys [
          {arriveTangent=2.7533, interpMode=RCIM_Cubic, leaveTangent=1.8958, tangentMode=RCTM_User, time=0, value=1}
          {arriveTangent=0.8292, interpMode=RCIM_Cubic, leaveTangent=0.8292, tangentMode=RCTM_User, time=0.321, value=1.3453}
          {arriveTangent=0.0529, interpMode=RCIM_Cubic, leaveTangent=0.2021, tangentMode=RCTM_User, time=1, value=1.5}
         ]}}}
      CurveIndex=&Add_Float(A=&Multiply_Float(A=$Particles.Initial.MaterialRandom, B=0.5), B=$Particles.NormalizedAge)
     )
    )
    Index 0 Param 2=&FloatFromCurve(FloatCurve=DataInterface() {propertyValues {Curve {keys [
         {arriveTangent=0.0453, interpMode=RCIM_Cubic, tangentMode=RCTM_Break, time=0, value=-1}
         {arriveTangent=0.0604, interpMode=RCIM_Cubic, leaveTangent=0.0604, tangentMode=RCTM_User, time=1, value=-1}
        ]}}})
    Index 0 Param 3=&FloatFromCurve(
     FloatCurve=DataInterface() {propertyValues {Curve {keys [
         {arriveTangent=0.0632, interpMode=RCIM_Cubic, leaveTangent=0.0632, tangentMode=RCTM_User, time=0.0591, value=6.0223}
         {arriveTangentWeight=0.1027, interpMode=RCIM_Cubic, tangentMode=RCTM_Break, time=0.2844, value=0}
        ]}}}
     Scale Curve=&RandomRangeFloat(Minimum=0.25, Maximum=0.5)
    )
    Index 0 Param 4=&FloatFromCurve(FloatCurve=DataInterface() {propertyValues {Curve {keys [<at 0.5 1>,<at 1 0>]}}})
    Index 1 Param 1=&FloatFromCurve(
     FloatCurve=DataInterface() {propertyValues {Curve {keys [
         {interpMode=RCIM_Cubic, tangentMode=RCTM_Break, time=0, value=1}
         {arriveTangentWeight=0.0833, interpMode=RCIM_Cubic, tangentMode=RCTM_Break, time=0.25, value=0.3}
        ]}}}
     Scale Curve=&RandomRangeFloat(Minimum=1.5, Maximum=3)
    )
    Index 1 Param 2=&FloatFromCurve(
     FloatCurve=DataInterface() {propertyValues {Curve {keys [
         {interpMode=RCIM_Cubic, leaveTangent=-2.5489, tangentMode=RCTM_Break, time=0, value=1}
         {interpMode=RCIM_Cubic, leaveTangent=0.9066, tangentMode=RCTM_Break, time=1, value=0}
        ]}}}
     Scale Curve=&RandomRangeFloat(Minimum=0.01, Maximum=0.1)
    )
    Index 1 Param 3=&FloatFromCurve(FloatCurve=DataInterface() {propertyValues {Curve {keys [
         {interpMode=RCIM_Cubic, leaveTangent=6.3917, tangentMode=RCTM_Break, time=0, value=0.2}
         {arriveTangent=1.7549, interpMode=RCIM_Cubic, leaveTangent=10.6767, tangentMode=RCTM_User, time=0.25, value=1}
        ]}}})
    Index 1 Param 4=&FloatFromCurve(
     FloatCurve=DataInterface() {propertyValues {Curve {keys [{arriveTangent=-0.136, interpMode=RCIM_Cubic, leaveTangent=-0.136, tangentMode=RCTM_User, time=0, value=1}, <at 1 0.95>]}}}
     Scale Curve=&RandomRangeFloat(Minimum=1.3, Maximum=2)
    )
    Index 2 Param 1=&FloatFromCurve(FloatCurve=DataInterface() {propertyValues {Curve {keys [
         {interpMode=RCIM_Cubic, leaveTangent=4.2485, tangentMode=RCTM_Break, time=0, value=0}
         {arriveTangent=1.0548, interpMode=RCIM_Cubic, leaveTangent=1.0548, tangentMode=RCTM_User, time=0.17, value=0.5715}
         <at 1 1>
        ]}}})
    Index 2 Param 2=&FloatFromCurve(
     FloatCurve=DataInterface() {propertyValues {Curve {keys [
         {arriveTangent=1.9056, interpMode=RCIM_Cubic, leaveTangent=1.9056, tangentMode=RCTM_User, time=0, value=0.2}
         {arriveTangent=0.0259, interpMode=RCIM_Cubic, leaveTangent=0.0259, tangentMode=RCTM_User, time=1, value=1.4}
        ]}}}
     Scale Curve=&RandomRangeFloat(Minimum=0.1, Maximum=0.2)
    )
    Index 2 Param 3=&FloatFromCurve(
     FloatCurve=DataInterface() {propertyValues {Curve {keys [
         {interpMode=RCIM_Cubic, leaveTangent=0.0393, tangentMode=RCTM_Break, time=0, value=1}
         {interpMode=RCIM_Cubic, leaveTangent=11.1233, tangentMode=RCTM_Break, time=1, value=1.5}
        ]}}}
     Scale Curve=0.7
    )
    Index 2 Param 4=&FloatFromCurve(
     FloatCurve=DataInterface() {propertyValues {Curve {keys [
         {arriveTangent=-0.6867, interpMode=RCIM_Cubic, leaveTangent=-0.6867, tangentMode=RCTM_User, time=0, value=1}
         {interpMode=RCIM_Cubic, leaveTangent=-1.2956, tangentMode=RCTM_Break, time=1, value=0.75}
        ]}}}
     Scale Curve=1.5
    )
    Module.Param0WriteEnabled=true
    Module.Param0WriteEnabled1=true
    Module.Param0WriteEnabled2=true
    Module.Param1WriteEnabled=false
    Module.Param1WriteEnabled1=true
    Module.Param1WriteEnabled2=true
    Module.Param2WriteEnabled=true
    Module.Param2WriteEnabled1=true
    Module.Param2WriteEnabled2=true
    Module.Param3WriteEnabled=true
    Module.Param3WriteEnabled1=true
    Module.Param3WriteEnabled2=true
    Write Parameter Index 1=true
    Write Parameter Index 2=true
   ) {module=DynamicMaterialParameters001}
   &FN_Sprite_Attributes(
    Color=&MakeLinearColorFromVectorAndFloat(
     "Vector (RGB)"=&Multiply_Vector(
      A=&MakeVectorFromLinearColorRGB(Color=<rgba 2 1.7786 1.7708 1>)
      B=&Multiply_VectorByFloat(
       Vector=&MakeVectorFromLinearColorRGB(Color=&ColorFromCurve(ColorCurve=DataInterface() {propertyValues {
           AlphaCurve {keys [<at 0 1 RCIM_Cubic>,<at 1 0 RCIM_Cubic>]}
           BlueCurve {keys [<at 0 0.0219 RCIM_Cubic>,<at 0.2893 0.0016 RCIM_Cubic>]}
           GreenCurve {keys [<at 0 0.0822 RCIM_Cubic>,<at 0.2893 0.0263 RCIM_Cubic>]}
           RedCurve {keys [<at 0 0.6302 RCIM_Cubic>,<at 0.2893 0.1354 RCIM_Cubic>]}
          }}))
       Float=&FloatFromCurve(FloatCurve=DataInterface() {propertyValues {Curve {keys [<at 0 1>,<at 1 0>]}}}, Scale Curve=6)
      )
     )
     "Float (Alpha)"=&Multiply_Float(
      A=&FloatFromCurve(
       FloatCurve=DataInterface() {propertyValues {Curve {keys [
           <at 0 0 RCIM_Cubic>
           {arriveTangent=3.0286, interpMode=RCIM_Cubic, leaveTangent=3.0286, time=0.1604, value=0.0265}
           {arriveTangent=2.048, interpMode=RCIM_Cubic, leaveTangent=2.048, time=0.3302, value=1}
           {arriveTangent=-1.4929, interpMode=RCIM_Cubic, leaveTangent=-1.4929, time=0.4892, value=0.7}
           <at 1 0>
          ]}}}
       Scale Curve=0.7
      )
      B=0.5
     )
    )
    Create Color Attribute=true
    Write to Module Namespace=true
   )
   &ScaleColor(Scale Alpha=$System.AgeFade, ScaleA=true, ScaleRGB=true) {module=ScaleColor001}
  ]
  rendererClasses [&NiagaraSpriteRendererProperties]
  renderers [
   {rendererClass=&NiagaraSpriteRendererProperties}
   {rendererClass=&NiagaraSpriteRendererProperties, rendererIndex=1}
  ]
 }
}
summary {
 systemName=NS_Emote_TalonScone_Fire
 emitters [
  {
   emitterName=Glow
   rendererClasses [
    &NiagaraSpriteRendererProperties
    &NiagaraLightRendererProperties
    &NiagaraMeshRendererProperties
   ]
  }
  {emitterName=Embers, rendererClasses [&NiagaraSpriteRendererProperties]}
  {emitterName=Smoke, rendererClasses [&NiagaraSpriteRendererProperties]}
  {emitterName=MainFire, rendererClasses [&NiagaraSpriteRendererProperties]}
  {emitterName=FiresprintSubUV, rendererClasses [&NiagaraSpriteRendererProperties]}
 ]
 userVariables [
  User.bIsFrontEnd {type=NiagaraBool}
  User.bIsFrontEndPreview {type=NiagaraBool}
  User.HotdogDistance {type=NiagaraFloat, defaultValue=1, description=""}
  User.LifeMultiplier {type=NiagaraFloat, defaultValue=0.56, description=""}
  User.VelocityMultiplier {type=NiagaraFloat, defaultValue=1.2, description=""}
 ]
}
