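# Animation Blueprint for the TalonScone quadruped cosmetic companion, copying its pose from another mesh and driving a hot-dog look-at and fire emote.
# CCDIK bends spine-to-head, weighted by the LookatHotDog curve and an interpolated blend.
# Each update reads the Contextual Anim IK target, caching it in actor space for when it disappears.
# With bUseReverseRetarget set, emote curves spawn and deactivate anticipation and fire Niagara effects, feeding fire the target distance.
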AnimBlueprint /CosmeticCompanions/Assets/Quadruped/TalonScone/Meshes/TalonScone_Quadruped_AnimBP
target_skeleton=&/TalonScone_Quadruped_Skeleton
graphs [
 AnimGraph {nodes [
   Blend Poses by bool(
    BlendPose_1=Copy Pose From Mesh() {bCopyCurves=true, bCopyCustomAttributes=true, bUseAttachedParent=true}
    BlendTime_0=0.4
    BlendTime_1=0.4
    bActiveValue=Get bUseReverseRetarget() -> (bUseReverseRetarget:bool)
   )
   n1 = CCDIK!(
    ComponentPose:Component Space Pose Link Structure=Local To Component(LocalPose=Copy Pose From Mesh() {bCopyCurves=true, bUseAttachedParent=true})
    Alpha:float=float * float(A=Get IKBlendWeight() -> (IKBlendWeight:double), B=GetCurveValue(CurveName=LookatHotDog)) {is=title K2Node_PromotableOperator}
    EffectorLocation:Vector="0,0,0"
   ) -> (Pose:Component Space Pose Link Structure) {
    is=title AnimGraphNode_CCDIK
    bEnableRotationLimit=true
    effectorLocationSpace=BCS_WorldSpace
    rootBone {boneName=spine_01}
    rotationLimitPerJoints [10,10,10,25,65,65]
    tipBone {boneName=head}
   }
   Output Pose(Result=Component To Local(ComponentPose=@n1))
  ]}
 EventGraph {nodes [
   n1 = Event BlueprintUpdateAnimation()
   > Set DeltaTime(DeltaTime:double=@n1.DeltaTimeX) -> (Output_Get:double)
   > GetIkTargetfromCAS()
   > Niagara()
  ]}
 GetIkTargetfromCAS {nodes [
   GetIkTargetfromCAS() {is=title K2Node_FunctionEntry}
   n1 = GetComponentByClass!(
    self:Actor Object Reference=GetOwningActor()
    ComponentClass:Actor Component Class Reference=/Script/ContextualAnimation.ContextualAnimSceneActorComponent
   ) -> (ReturnValue:Contextual Anim Scene Actor Component Object Reference)
   n2 = BreakTransform(InTransform=Break Contextual Anim IKTarget!(ContextualAnimIKTarget:"Contextual Anim IKTarget Structure (by ref)"=GetIKTargetByGoalName!(self:Contextual Anim Scene Actor Component Object Reference=@n1, GoalName:Name=LookatHotDog) -> (ReturnValue:"Contextual Anim IKTarget Structure (by ref)")) -> (GoalName:Name, BoneName:Name, Alpha:float=~, Transform:Transform) {is=title K2Node_BreakStruct}.Transform)
   > n3 = Branch(Condition="Not Equal (Float)"(A=Vector Length(A=@n2)))
   n4 = Get Actor Transform(self=GetOwningActor())
   > Set LookAtLocalOffset(LookAtLocalOffset:Vector=InverseTransformLocation!(T:"Transform (by ref)"=@n4, Location:Vector=@n2) -> (ReturnValue:Vector)) -> (Output_Get:Vector)
   n5 = TransformLocation!(T:"Transform (by ref)"=@n4, Location:Vector=Get LookAtLocalOffset() -> (LookAtLocalOffset:Vector)) -> (ReturnValue:Vector)
   > Set LookAtTargetLocation(LookAtTargetLocation:Vector=@n5) -> (Output_Get:Vector)
   > Set IKBlendWeight(IKBlendWeight:double=FInterpTo(Current=Get IKBlendWeight(), Target=1, DeltaTime=Get DeltaTime() -> (DeltaTime:double), InterpSpeed=5)) -> (Output_Get:double)
   Set LookAtTargetLocation(execute=@n3.else, LookAtTargetLocation=@n5)
   > Set IKBlendWeight(IKBlendWeight=FInterpTo(Current=Get IKBlendWeight(), Target=0, DeltaTime=Get DeltaTime(), InterpSpeed=2))
  ]}
 Niagara {nodes [
   Niagara() {is=title K2Node_FunctionEntry}
   > n1 = Branch(Condition=Get bUseReverseRetarget())
   > n2 = Sequence()
   > n3 = Branch(Condition="float > float"(A=GetCurveValue(CurveName=TalonSconeEmoteAnticipation)))
   > n4 = Branch(Condition=IsValid(Object=Get FireNiagaraAnticip() -> (FireNiagaraAnticip:Niagara Particle System Component Object Reference)))
   n5 = SpawnSystemAttached!(
    execute=@n4.else
    SystemTemplate:Niagara System Object Reference=/CosmeticCompanions/FX/Builtins/TalonScone/Emote/NS_Emote_TalonScone_FireAnticipation
    AttachToComponent:Scene Component Object Reference=GetOwningComponent()
    AttachPointName:Name
    Location:Vector="0, 0, 0"
    Rotation:Rotator="0, 0, 0"
    LocationType:EAttachLocation Enum=SnapToTarget
    bAutoDestroy:bool=true
    bAutoActivate:bool=true
    PoolingMethod:ENCPoolMethod Enum=None
    bPreCullCheck:bool=true
   ) -> (then, ReturnValue:Niagara Particle System Component Object Reference)
   > Set FireNiagaraAnticip(FireNiagaraAnticip:Niagara Particle System Component Object Reference=@n5) -> (Output_Get:Niagara Particle System Component Object Reference)
   Branch(execute=@n3.else, Condition=IsValid(Object=Get FireNiagaraAnticip()))
   > Deactivate(self:Actor Component Object Reference=Get FireNiagaraAnticip())
   > Set FireNiagaraAnticip()
   n6 = Branch(execute=@n2.then_1, Condition="float > float"(A=GetCurveValue(CurveName=TalonSconeEmoteFire)))
   > n7 = Branch(Condition=IsValid(Object=Get FireNiagaraComp() -> (FireNiagaraComp:Niagara Particle System Component Object Reference)))
   n8 = SpawnSystemAttached(
    execute=@n7.else
    SystemTemplate=/CosmeticCompanions/FX/Builtins/TalonScone/Emote/NS_Emote_TalonScone_Fire
    AttachToComponent=GetOwningComponent()
    AttachPointName=head
    Location="0,17,3"
    Rotation="0,-120,0"
    LocationType=SnapToTarget
    bAutoDestroy=true
    bAutoActivate=true
    PoolingMethod=None
    bPreCullCheck=true
   )
   > SetBoolParameter!(
    execute
    self:FXSystem Component Object Reference=@n8
    ParameterName:Name=bIsFrontEnd
    Param:bool=IsCosmeticObjectDisplayedInFrontendContext!(CosmeticObject:Object Reference) -> (ReturnValue:bool)
   ) -> (then)
   n9 = Get LookAtLocalOffset()
   > "Set Niagara Variable (Float)"!(
    execute
    self:Niagara Particle System Component Object Reference=@n8
    InVariableName:Name=HotdogDistance
    InValue:float=MapRangeClamped(Value=Vector Length(A=@n9), InRangeA=160, InRangeB=270)
   ) -> (then) {is=title K2Node_CallFunction}
   > Set FireNiagaraComp(FireNiagaraComp:Niagara Particle System Component Object Reference=@n8) -> (Output_Get:Niagara Particle System Component Object Reference)
   Branch(execute=@n6.else, Condition=IsValid(Object=Get FireNiagaraComp()))
   > Deactivate(self=Get FireNiagaraComp())
   > Set FireNiagaraComp()
   n10 = Sequence(execute=@n1.else)
   > Branch(Condition=IsValid(Object=Get FireNiagaraAnticip()))
   > Deactivate(self=Get FireNiagaraAnticip())
   > Set FireNiagaraAnticip()
   Branch(execute=@n10.then_1, Condition=IsValid(Object=Get FireNiagaraComp()))
   > Deactivate(self=Get FireNiagaraComp())
   > Set FireNiagaraComp()
  ]}
 OnInitialUpdate {nodes [
   OnInitialUpdate!() -> (then, Context:"Anim Update Context Structure (by ref)", Node:"Anim Node Reference Structure (by ref)") {is=title K2Node_FunctionEntry}
   > Return Node()
  ]}
]
variables [
 bFireActive {type=boolean}
 bHasValidIKTarget {type=boolean}
 DeltaTime {type=number}
 FireNiagaraAnticip {type=NiagaraComponent, default=null}
 FireNiagaraComp {type=NiagaraComponent, default=null}
 FireNiagaraRotation {type=Rotator, default=<rot 0 -120 0>}
 IKBlendWeight {type=number}
 InvisSkeletonMesh {type=SkeletalMeshComponent}
 LookAtLocalOffset {type=Vector}
 LookAtTargetLocation {type=Vector}
]
